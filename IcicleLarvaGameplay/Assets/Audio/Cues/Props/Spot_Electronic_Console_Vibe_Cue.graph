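# Spot ambient sound cue for an electronic console prop, routed to an indoor ambient sound class with looping-spot attenuation.
# A mixer layers two console vibe loop waves with a randomly selected hologram loop, one of three equally weighted variants.
# The hologram layer passes through a distance crossfade at reduced volume and fades out at mid range, so it is heard only close to the source.
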SoundCue /IcicleLarvaGameplay/Assets/Audio/Cues/Props/Spot_Electronic_Console_Vibe_Cue
root=Mixer_1
nodes [
 {
  id=DistanceCrossFade_1
  children [Random_0,null]
  properties {crossFadeInput [
    {fadeInDistanceEnd=0, fadeInDistanceStart=0, fadeOutDistanceEnd=800, fadeOutDistanceStart=500, volume=0.5}
    {fadeInDistanceEnd=0, fadeInDistanceStart=0, fadeOutDistanceEnd=0, fadeOutDistanceStart=0, volume=1}
   ]}
 }
 {id=Mixer_1, children [WavePlayer_4, WavePlayer_5, DistanceCrossFade_1], properties {inputVolume [1,1,1]}}
 {id=Random_0, children [WavePlayer_0,WavePlayer_6,WavePlayer_7], properties {weights [1,1,1]}}
 {id=WavePlayer_0, properties {soundWaveAssetPtr=/IcicleLarvaGameplay/Assets/Audio/Waves/Spot_HologramLoop_01}}
 {id=WavePlayer_4, properties {soundWaveAssetPtr=/IcicleLarvaGameplay/Assets/Audio/Waves/Spot_Electronic_Console_Vibe_Loop_2}}
 {id=WavePlayer_5, properties {soundWaveAssetPtr=/IcicleLarvaGameplay/Assets/Audio/Waves/Spot_Electronic_Console_Vibe_Loop_1}}
 {id=WavePlayer_6, properties {soundWaveAssetPtr=/IcicleLarvaGameplay/Assets/Audio/Waves/Spot_HologramLoop_02}}
 {id=WavePlayer_7, properties {soundWaveAssetPtr=/IcicleLarvaGameplay/Assets/Audio/Waves/Spot_HologramLoop_03}}
]
properties {
 attenuationSettings=/Game/Sounds/Attenuation/Fort_Looping_Spot_Att
 maxDistance=3750
 soundClassObject=&Fort_Ambient_Indoors_Athena
 volumeMultiplier=0.4
}
